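# Physics asset for the MED Mod Ninja backpack cosmetic, giving its two dangling straps physical bodies and joint limits.
# The pack body combines a capsule and a box, while each left and right strap is a chain of six small box bodies hanging from it.
# Limited swing and twist constraints link each segment to the one above, with twist and one swing axis locked on each end segment.
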PhysicsAsset /BRCosmetics/Accessories/FORT_Backpacks/Backpack_M_MED_Mod_Ninja/Meshes/M_MED_Mod_Ninja_Pack_Physics
bodies {
 dyn_l_strap_joint [{
   shapeType=Box
   extentX=1.1698
   extentY=1.01
   extentZ=3.1698
   shapeName=dyn_l_strap_joint_box
   center=<0.4488 -0.0455 0.0324>
   rotation=<rot 1.7776 -90.0153 -1.0613>
  }]
 dyn_l_strap_joint1 [{
   shapeType=Box
   extentX=1.1698
   extentY=1.01
   extentZ=3.0226
   shapeName=dyn_l_strap_joint1_box
   center=<0.0406 0.0219 -0.1273>
   rotation=<rot -0.1731 -89.9831 -0.6579>
  }]
 dyn_l_strap_joint2 [{
   shapeType=Box
   extentX=1.0898
   extentY=1.01
   extentZ=2.8614
   shapeName=dyn_l_strap_joint2_box
   center=<-0.0089 -0.0447 -0.0374>
   rotation=<rot -0.2176 -89.9799 -0.8431>
  }]
 dyn_l_strap_joint3 [{
   shapeType=Box
   extentX=1.4089
   extentY=1.1109
   extentZ=2.8865
   shapeName=dyn_l_strap_joint3_box
   center=<-0.1614 0.2699 -0.0434>
   rotation=<rot 1.6312 -89.9811 -0.8364>
  }]
 dyn_l_strap_joint4 [{
   shapeType=Box
   extentX=1.8078
   extentY=1.6479
   extentZ=2.8898
   shapeName=dyn_l_strap_joint4_box
   center=<0.2461 0.2911 -0.0893>
   rotation=<rot 0.4652 -90.0043 0.3435>
  }]
 dyn_l_strap_joint5 [{
   shapeType=Box
   extentX=2.1312
   extentY=3.9415
   extentZ=4.8801
   shapeName=dyn_l_strap_joint5_box
   center=<2.0947 -0.4062 -0.113>
   rotation=<rot 3.8359 75.3153 -89.265>
  }]
 dyn_r_strap_joint [{
   shapeType=Box
   extentX=1.01
   extentY=1.01
   extentZ=2.8579
   shapeName=dyn_r_strap_joint_box
   center=<0.6184 -0.0492 -0.0041>
   rotation=<rot 1.7613 -90.0184 -1.787>
  }]
 dyn_r_strap_joint1 [{
   shapeType=Box
   extentX=1.0456
   extentY=0.9293
   extentZ=2.8038
   shapeName=dyn_r_strap_joint1_box
   center=<0.1946 0.0557 -0.0228>
   rotation=<rot 3.6662 -90.0445 -1.2046>
  }]
 dyn_r_strap_joint2 [{
   shapeType=Box
   extentX=1.0265
   extentY=0.95
   extentZ=2.8614
   shapeName=dyn_r_strap_joint2_box
   center=<0.1136 -0.062 -0.0102>
   rotation=<rot 3.6316 -90.009 -0.3492>
  }]
 dyn_r_strap_joint3 [{
   shapeType=Box
   extentX=1.505
   extentY=1.0553
   extentZ=2.8865
   shapeName=dyn_r_strap_joint3_box
   center=<-0.0331 0.1293 -0.0521>
   rotation=<rot 4.4869 -89.9868 0.3388>
  }]
 dyn_r_strap_joint4 [{
   shapeType=Box
   extentX=2.0322
   extentY=1.4776
   extentZ=2.9441
   shapeName=dyn_r_strap_joint4_box
   center=<0.4289 0.3156 -0.0795>
   rotation=<rot 3.9953 -89.9813 0.3552>
  }]
 dyn_r_strap_joint5 [{
   shapeType=Box
   extentX=5.1743
   extentY=1.2765
   extentZ=4.1504
   shapeName=dyn_r_strap_joint5_box
   center=<2.6014 -0.275 -0.0877>
   rotation=<rot -0.0025 -11.4594 3.6772>
  }]
 pack [
  {shapeType=Capsule, length=24.4665, radius=10.3656, center=<1.5493 -12.1174 -0.0165>, rotation=<rot -1.9798 87.7267 -75.8269>}
  {
   shapeType=Box
   extentX=44.0346
   extentY=4.6299
   extentZ=7.5
   shapeName=pack_box
   center=<1.9376 -19.4308 -6.6222>
   rotation=<rot 15.4303 -1.0135 -34.1286>
  }
 ]
}
constraints [
 {bone1Name=dyn_l_strap_joint, bone2Name=pack, swing1LimitDegrees=45, swing2LimitDegrees=25, twistLimitDegrees=10}
 {bone1Name=dyn_l_strap_joint1, bone2Name=dyn_l_strap_joint, swing1LimitDegrees=30, swing2LimitDegrees=30, twistLimitDegrees=10}
 {bone1Name=dyn_l_strap_joint2, bone2Name=dyn_l_strap_joint1, swing1LimitDegrees=30, swing2LimitDegrees=30, twistLimitDegrees=10}
 {bone1Name=dyn_l_strap_joint3, bone2Name=dyn_l_strap_joint2, swing1LimitDegrees=30, swing2LimitDegrees=30, twistLimitDegrees=10}
 {bone1Name=dyn_l_strap_joint4, bone2Name=dyn_l_strap_joint3, swing1LimitDegrees=30, swing2LimitDegrees=30, twistLimitDegrees=10}
 {
  bone1Name=dyn_l_strap_joint5
  bone2Name=dyn_l_strap_joint4
  swing1LimitDegrees=30
  swing2LimitDegrees=25
  swing2Motion=Locked
  twistLimitDegrees=10
  twistMotion=Locked
 }
 {bone1Name=dyn_r_strap_joint, bone2Name=pack, swing1LimitDegrees=45, swing2LimitDegrees=25, twistLimitDegrees=10}
 {bone1Name=dyn_r_strap_joint1, bone2Name=dyn_r_strap_joint, swing1LimitDegrees=30, swing2LimitDegrees=30, twistLimitDegrees=10}
 {bone1Name=dyn_r_strap_joint2, bone2Name=dyn_r_strap_joint1, swing1LimitDegrees=30, swing2LimitDegrees=30, twistLimitDegrees=10}
 {bone1Name=dyn_r_strap_joint3, bone2Name=dyn_r_strap_joint2, swing1LimitDegrees=30, swing2LimitDegrees=30, twistLimitDegrees=10}
 {bone1Name=dyn_r_strap_joint4, bone2Name=dyn_r_strap_joint3, swing1LimitDegrees=30, swing2LimitDegrees=30, twistLimitDegrees=10}
 {
  bone1Name=dyn_r_strap_joint5
  bone2Name=dyn_r_strap_joint4
  swing1LimitDegrees=30
  swing2LimitDegrees=45
  swing2Motion=Locked
  twistLimitDegrees=45
  twistMotion=Locked
 }
]
